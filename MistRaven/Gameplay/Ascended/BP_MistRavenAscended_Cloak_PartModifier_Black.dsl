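; Part modifier Blueprint for the black cloak of the Mist Raven Ascended that gives the part a Wraith Form material on the player pawn.
; On the post-res-in event it calls the parent logic, then overrides the first two material slots of the part's skeletal mesh with the MI_WraitForm_MistRaven instance.
; The construction script only calls its parent, and nothing here restores the original materials.
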(event EventOnPostResIn
  (bind _returnvalue (CustomCharacter|GetSkeletalMeshForPartOnPlayerPawn))
  (|CallParentFunction)
  (Rendering|Material|SetMaterial _returnvalue 0 "/MistRaven/Assets/Materials/MI_WraitForm_MistRaven.MI_WraitForm_MistRaven")
  (Rendering|Material|SetMaterial _returnvalue 1 "/MistRaven/Assets/Materials/MI_WraitForm_MistRaven.MI_WraitForm_MistRaven"))

(fn ConstructionScript
  (|CallParentFunction))
