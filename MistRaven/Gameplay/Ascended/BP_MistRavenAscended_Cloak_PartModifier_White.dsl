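; Cloak part modifier for MistRaven's white Ascended look, applying wraith-form materials to the part's skeletal mesh after res-in.
; If the player pawn has the ascended-achieved tag, it plays a cape glow sound and puts an animated dynamic material on both slots, stamped with game time.
; A cloak animation timer flag stays on for a short delay.
; The slots then switch to the static white material, a cheaper version of the animated one.
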(event EventOnPostResIn
  (bind _returnvalue (CustomCharacter|GetSkeletalMeshForPartOnPlayerPawn))
  (bind _returnvalue_2 (CustomCharacter|GetSkeletalMeshForPartOnPlayerPawn))
  (|CallParentFunction)
  (if (GameplayTags|HasMatchingGameplayTag (CustomCharacter|GetAssociatedPlayerPawn) "(TagName=\"MistRaven.AscendedAcheived\")")
    (Audio|SpawnSoundAttached "/MistRaven/Assets/Audio/MSS_MistRaven_KitComplete_CapeGlow.MSS_MistRaven_KitComplete_CapeGlow" _returnvalue_2)
    (bind _returnvalue_3 (Rendering|Material|CreateDynamicMaterialInstance "/MistRaven/Assets/Materials/MI_WraitForm_MistRaven_Animation.MI_WraitForm_MistRaven_Animation" "None" "None" :!class /Script/Engine.KismetMaterialLibrary)) ; assign and play animation material
    (Rendering|Material|SetMaterial _returnvalue_2 0 _returnvalue_3)
    (Rendering|Material|SetMaterial _returnvalue_2 1 _returnvalue_3)
    (Rendering|Material|SetScalarParameterValue _returnvalue_3 "TL_Time" (Utilities|Time|GetGameTimeinSeconds) :!class /Script/Engine.MaterialInstanceDynamic)
    (Variables|Default|SetCloakAnimTimerRunning true)
    (Utilities|FlowControl|Delay 2.0) ; Switch material to cheap version
    (Variables|Default|SetCloakAnimTimerRunning false)
    (else
      _))
  (Rendering|Material|SetMaterial _returnvalue 0 "/MistRaven/Assets/Materials/MI_WraitForm_MistRaven_White.MI_WraitForm_MistRaven_White")
  (Rendering|Material|SetMaterial _returnvalue 1 "/MistRaven/Assets/Materials/MI_WraitForm_MistRaven_White.MI_WraitForm_MistRaven_White"))

(fn ConstructionScript
  (|CallParentFunction))
