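; Part modifier Blueprint for the MistRaven Ascended right glove, named as the Black variant, that overrides a character part's material.
; On the post res-in event it runs the parent logic, then gets the part's skeletal mesh on the player pawn.
; It sets that mesh's material slot 0 to the hard-referenced MI_Flaemthrower_MistRaven material instance.
; The construction script only calls its parent.
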(event EventOnPostResIn
  (|CallParentFunction)
  (Rendering|Material|SetMaterial (CustomCharacter|GetSkeletalMeshForPartOnPlayerPawn) 0 "/MistRaven/Assets/Materials/MI_Flaemthrower_MistRaven.MI_Flaemthrower_MistRaven"))

(fn ConstructionScript
  (|CallParentFunction))
