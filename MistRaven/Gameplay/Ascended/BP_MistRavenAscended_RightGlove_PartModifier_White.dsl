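; Part modifier for the white Mist Raven Ascended right glove that applies its material to the player's part mesh after the post-res-in event.
; If the player pawn has the MistRaven.AscendedAcheived tag, it plays a timed reveal using a dynamic animation material stamped with the current game time.
; A cloak animation timer flag stays on during the two-second delay.
; The mesh then gets the cheaper static white material, which untagged players receive immediately.
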(event EventOnPostResIn
  (|CallParentFunction)
  (if (GameplayTags|HasMatchingGameplayTag (CustomCharacter|GetAssociatedPlayerPawn) "(TagName=\"MistRaven.AscendedAcheived\")")
    (bind _returnvalue (Rendering|Material|CreateDynamicMaterialInstance "/MistRaven/Assets/Materials/MI_Flaemthrower_MistRaven_Animation.MI_Flaemthrower_MistRaven_Animation" "None" "None" :!class /Script/Engine.KismetMaterialLibrary)) ; assign and play animation material
    (Rendering|Material|SetMaterial (CustomCharacter|GetSkeletalMeshForPartOnPlayerPawn) 0 _returnvalue)
    (Rendering|Material|SetScalarParameterValue _returnvalue "TL_Time" (Utilities|Time|GetGameTimeinSeconds) :!class /Script/Engine.MaterialInstanceDynamic)
    (Variables|Default|SetCloakAnimTimerRunning true)
    (Utilities|FlowControl|Delay 2.0) ; Switch material to cheap version
    (Variables|Default|SetCloakAnimTimerRunning false)
    (else
      _))
  (Rendering|Material|SetMaterial (CustomCharacter|GetSkeletalMeshForPartOnPlayerPawn) 0 "/MistRaven/Assets/Materials/MI_Flaemthrower_MistRaven_White.MI_Flaemthrower_MistRaven_White"))

(fn ConstructionScript
  (|CallParentFunction))
