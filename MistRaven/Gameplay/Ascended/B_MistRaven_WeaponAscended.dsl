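; Weapon Blueprint for the Ascended Mist Raven, managing ability-driven HUD and visuals.
; On equip, chained gameplay-tag wait tasks toggle Grim Strut, Tendril Drain and Connector key hints, mesh visibility and an all-equipped material, plus a timed first-time HUD flicker.
; The Grim Strut active tag toggles ascend/descend hints and arms a cooldown-exit sound.
; It also removes itself when unequipped empty, destroys flagged matching pickups and stores replicated-style tether points.
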(event Custom|TryReload)

(event FortWeapon|EventOnUnEquippedAndDetached
  ;[1 1 1] On Unequipped
  (|CallParentFunction)
  ;[0.15 0.002451 0.03353 0.5] Solving Bug where if player tries to pick up a new Mist Raven w/ 1 already equipped
  ; Check and add tag "MistRaven.GrimStrut.Ascended.InEquipment" - Kentaro Tamai 10/28/25
  (Ability|GameplayEffect|RemoveGameplayTags (Game|GetInstigator) (Variables|Default|GetAddGrimStrutAscendedInEquipmentTag))
  (CallFunction|CheckRemoveWeapononEmpty)
  (CallFunction|UnbindControllerPickupEvent))

(event OnAnimInitialized(WeaponMesh))

(event FortWeapon|EventOnEquippedAndAttached
  (bind _returnvalue (Game|GetInstigator))
  (bind _returnvalue_2 (Game|GetInstigator))
  (bind _returnvalue_3 (Game|GetInstigator))
  ;[1 1 1] On Equipped
  (|CallParentFunction)
  (Rendering|Material|CreateDynamicMaterialInstance (Variables|Visual|GetWeaponMesh) 0 (Rendering|Material|GetMaterial (Variables|Visual|GetWeaponMesh) 0) "None" :!class /Script/Engine.PrimitiveComponent)
  (bind _asfort_player_pawn (Utilities|Casting|CastToFortPlayerPawn (Game|GetInstigator)))
  (CallFunction|BindtoControllerPickupEvent)
  ;;[1 1 1] Mist Raven Visuals based on inventory tags
  (Ability|Tasks|WaitGameplayTagAddToActor _returnvalue "(TagName=\"MistRaven.GrimStrut.Ascended.InEquipment\")"
    (:then
      (Ability|Tasks|WaitGameplayTagAddToActor _returnvalue "(TagName=\"MistRaven.TendrilDrain.Ascended.InEquipemtn\")"
        (:then
          (Ability|Tasks|WaitGameplayTagAddToActor _returnvalue "(TagName=\"MistRaven.Connector.Ascended.InEquipment\")"
            (:then
              (Ability|Tasks|WaitGameplayTagRemoveFromActor _returnvalue_2 "(TagName=\"MistRaven.TendrilDrain.Ascended.InEquipemtn\")"
                (:then
                  (Ability|Tasks|WaitGameplayTagRemoveFromActor _returnvalue_2 "(TagName=\"MistRaven.Connector.Ascended.InEquipment\")"
                    (:then
                      (Ability|Tasks|WaitGameplayTagRemoveFromActor _returnvalue_2 "(TagName=\"MistRaven.GrimStrut.Ascended.InEquipment\")"
                        (:then
                          (Ability|Tasks|WaitGameplayTagAddToActor _returnvalue_3 "(TagName=\"MistRaven.AllAscendedEquipped\")"
                            (:then
                              (Ability|Tasks|WaitGameplayTagRemoveFromActor _returnvalue_3 "(TagName=\"MistRaven.AllAscendedEquipped\")"
                                (:then
                                  (Utilities|FlowControl|DoOnce false
                                    (:Completed
                                      (bind _returnvalue_4 (GameplayTags|HasMatchingGameplayTag(Message) (Game|GetInstigator) "(TagName=\"MistRaven.AscendedAcheived\")"))
                                      (if _returnvalue_4
                                        (CallFunction|DoFlickeringHUD)))))
                                (:Removed
                                  (Rendering|Material|SetMaterial (Variables|Visual|GetWeaponMesh) 1 "/Game/Athena/Prototype/HardSurface/Materials/PacGunMats/HS_PAC_RedPlastic.HS_PAC_RedPlastic"))))
                            (:Added
                              (Rendering|Material|SetMaterial (Variables|Visual|GetWeaponMesh) 1 "/Game/Athena/Prototype/HardSurface/Materials/PacGunMats/HS_PAC_PurplePlastic.HS_PAC_PurplePlastic"))))
                        (:Removed
                          (WeaponHUDData|SetHudKeyActionVisibilitybyID "WraithForm")
                          (WeaponHUDData|SetHudKeyActionVisibilitybyID "WraithFormPrimary"))))
                    (:Removed
                      (Rendering|SetVisibility (Variables|Visual|GetWeaponMesh) false true)
                      (WeaponHUDData|SetHudKeyActionVisibilitybyID "Teleporter"))))
                (:Removed
                  (WeaponHUDData|SetHudKeyActionVisibilitybyID "Flamethrower"))))
            (:Added
              (Utilities|Casting|CastToFortPlayerPawn _returnvalue)
              (Rendering|SetVisibility (Variables|Visual|GetWeaponMesh) true true)
              (WeaponHUDData|SetHudKeyActionVisibilitybyID "Teleporter" true))))
        (:Added
          (WeaponHUDData|SetHudKeyActionVisibilitybyID "Flamethrower" true)
          (WeaponHUDData|SetHudKeyActionVisibilitybyID "WraithFormPrimary"))))
    (:Added
      (WeaponHUDData|SetHudKeyActionVisibilitybyID "WraithForm" true)
      (bind _returnvalue_5 (GameplayTags|HasMatchingGameplayTag(Message) _returnvalue "(TagName=\"MistRaven.TendrilDrain.Ascended.InEquipemtn\")"))
      (if (not _returnvalue_5)
        (WeaponHUDData|SetHudKeyActionVisibilitybyID "WraithFormPrimary" true)))))

(event EventBeginPlay
  ;[1 1 1] Begin Play
  (|CallParentFunction)
  (Variables|Default|SetDeleteItemLater false) ; reset in case of pooling
  (Ability|Tasks|WaitGameplayTagAddToActor (Game|GetInstigator) "(TagName=\"MistRaven.GrimStrut.Active\")"
    (:then
      (Ability|Tasks|WaitGameplayTagRemoveFromActor (Game|GetInstigator) "(TagName=\"MistRaven.GrimStrut.Active\")"
        (:then
          (Rendering|SetVisibility (Variables|Visual|GetWeaponMesh) false true)
          (Variables|Default|SetFirstTimeEquip false))
        (:Removed
          (CallFunction|DeactivateGrimStrut))))
    (:Added
      (CallFunction|ActivateGrimStrut))))

(event Custom|ActivateGrimStrut
  (WeaponHUDData|SetHudKeyActionVisibilitybyID "WraithFormAscend" true)
  (WeaponHUDData|SetHudKeyActionVisibilitybyID "WraithFormDescend" true))

(event Custom|DeactivateGrimStrut
  (bind _returnvalue (Game|GetInstigator))
  (WeaponHUDData|SetHudKeyActionVisibilitybyID "WraithFormDescend")
  (WeaponHUDData|SetHudKeyActionVisibilitybyID "WraithFormAscend")
  (Ability|Tasks|WaitGameplayTagRemoveFromActor _returnvalue "(TagName=\"Granted.Cooldown.MistRaven.GrimStrut.Ascended.Cooldown\")"
    (:Removed
      ;[1 1 1] On Equipped
      (if (and (FortWeapon|IsEquipped) (Pawn|IsLocallyViewed _returnvalue))
        (Audio|PlaySound2D "/MistRaven/Assets/Audio/MSS_MistRaven_GrimStrut_UI_ExitCooldown.MSS_MistRaven_GrimStrut_UI_ExitCooldown")))))

(event EventOnUnableToPerformAction
  (params (FailedReason :type Gameplay%20Tag%20Container%20Structure%20%28by%20ref%29) (LocalizedFailureReasonText :type Text%20%28by%20ref%29))
  (CallFunction|PlayUICannotDoSound FailedReason))

(event Custom|BindToControllerPickupEvent
  (bind _asfort_player_controller (Utilities|Casting|CastToFortPlayerController (Pawn|GetController (Game|GetInstigator))))
  (EventDispatchers|BindEventtoOnPickupCreated _asfort_player_controller (delegate Custom|PickupCreated)))

(event Custom|PickupCreated
  (params (PlayerController :type Fort%20Player%20Controller%20Object%20Reference) (Pickup :type Fort%20Pickup%20Object%20Reference))
  (Utilities|IsValid Pickup
    (:Is%20Valid
      (if (and (or (GameplayTags|HasMatchingGameplayTag (Pickup|GetPrimaryItemDefinition Pickup) (Variables|Default|GetGrimStrutWIDTag)) (GameplayTags|HasMatchingGameplayTag (Pickup|GetPrimaryItemDefinition Pickup) (Variables|Default|GetCombinedTag))) (Variables|Default|GetDeleteItemLater))
        (Actor|DestroyActor Pickup)))))

(event Custom|Multi_StartEndPos
  (params (StartPos :type Vector) (EndPos :type Vector))
  ;[1 1 1] Tendril Drain Tether Visuals
  (Variables|Default|SetTendrilStartPoint StartPos)
  (Variables|Default|SetTendrilEndPoint EndPos))

(event Custom|Multi_TetherAttachPawn
  (params (AttachTarget :type Actor%20Object%20Reference))
  (Variables|Default|SetTendrilAttachTarget AttachTarget))

(event Custom|UnbindControllerPickupEvent
  (bind _asfort_player_controller (Utilities|Casting|CastToFortPlayerController (Pawn|GetController (Game|GetInstigator))))
  (EventDispatchers|UnbindEventfromOnPickupCreated _asfort_player_controller (delegate Custom|PickupCreated)))

(event EventOnWeaponVisibilityChanged
  (params (bVisible :type Boolean) (bSetForLocalControllerOnly :type Boolean)))

(event Custom|DoFlickeringHUD
  ;[1 1 1] UI HUD flicker the first time we transform inputs
  (Utilities|Time|SetTimerbyEvent (delegate Custom|Flicker_one) (Variables|Default|GetFlickerTimer))
  (Utilities|Time|SetTimerbyEvent (delegate Custom|FlickerTwo) (* (Variables|Default|GetFlickerTimer) 2.0))
  (Utilities|Time|SetTimerbyEvent (delegate Custom|FlickerThree) (* (Variables|Default|GetFlickerTimer) 3.0))
  (Utilities|Time|SetTimerbyEvent (delegate Custom|FlickerFour) (* (Variables|Default|GetFlickerTimer) 4.0))
  (Utilities|Time|SetTimerbyEvent (delegate Custom|FlickerFive) (* (Variables|Default|GetFlickerTimer) 5.0))
  (Utilities|Time|SetTimerbyEvent (delegate Custom|FlickerSix) (* (Variables|Default|GetFlickerTimer) 6.0)))

(event Custom|Flicker_one
  ;[1 1 1] UI HUD flicker the first time we transform inputs
  (WeaponHUDData|SetHudKeyActionVisibilitybyID "WraithForm")
  (WeaponHUDData|SetHudKeyActionVisibilitybyID "Flamethrower"))

(event Custom|FlickerTwo
  ;[1 1 1] UI HUD flicker the first time we transform inputs
  (WeaponHUDData|SetHudKeyActionVisibilitybyID "WraithForm" true)
  (WeaponHUDData|SetHudKeyActionVisibilitybyID "Flamethrower" true))

(event Custom|FlickerThree
  ;[1 1 1] UI HUD flicker the first time we transform inputs
  (WeaponHUDData|SetHudKeyActionVisibilitybyID "WraithForm")
  (WeaponHUDData|SetHudKeyActionVisibilitybyID "Flamethrower"))

(event Custom|FlickerFour
  ;[1 1 1] UI HUD flicker the first time we transform inputs
  (WeaponHUDData|SetHudKeyActionVisibilitybyID "WraithForm" true)
  (WeaponHUDData|SetHudKeyActionVisibilitybyID "Flamethrower" true))

(event Custom|FlickerFive
  ;[1 1 1] UI HUD flicker the first time we transform inputs
  (WeaponHUDData|SetHudKeyActionVisibilitybyID "WraithForm")
  (WeaponHUDData|SetHudKeyActionVisibilitybyID "Flamethrower"))

(event Custom|FlickerSix
  ;[1 1 1] UI HUD flicker the first time we transform inputs
  (WeaponHUDData|SetHudKeyActionVisibilitybyID "WraithForm" true)
  (WeaponHUDData|SetHudKeyActionVisibilitybyID "Flamethrower" true))

(event Custom|CheckRemoveWeaponOnEmpty
  ;[1 1 1] Check Remove Weapon On Empty
  (if (<= (FortWeapon|GetLocalRemainingAmmo) 0)
    (Inventory|RemoveItemFromPlayerByGuid (Fortnite|GetFortPlayerControllerfromActor (Game|GetInstigator)) (FortWeapon|GetInventoryGUID) 1 true)))

(fn PlayUICannotDoSound
  (params (TagContainer :type Gameplay%20Tag%20Container%20Structure%20%28by%20ref%29))
  (bind _returnvalue (Game|GetInstigator))
  (bind _returnvalue_2 (GameplayTags|HasMatchingGameplayTag(Message) _returnvalue "(TagName=\"MistRaven.GrimStrut.Active\")"))
  (if (Math|Boolean|ANDBoolean (not _returnvalue_2) (Pawn|IsLocallyViewed _returnvalue) (GameplayTags|HasAnyTags TagContainer (Variables|Default|GetActiveate%20Failed%20Tag)))
    (Audio|PlaySound2D "/FNGameplayCues/Sounds/Fort_UI_Sounds/Gameplay/UI_Error_Ping_Cant_Do_Default_Cue.UI_Error_Ping_Cant_Do_Default_Cue")))

(fn ConstructionScript)
