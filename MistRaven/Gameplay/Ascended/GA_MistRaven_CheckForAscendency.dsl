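; Gameplay ability for MistRaven that checks whether the player has reached ascension through Ascended equipment.
; It sets up tag add and remove wait tasks for TendrilDrain, Connector and GrimStrut, alongside a step clearing the all-ascended loose tags.
; When all required tags are present without the all-ascended tag, it adds that tag.
; It then plays the AscensionComplete montage if any Ascended item is equipped and, after a delay, removes the achieved tag.
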(event Ability|EventActivateAbility
  (bind _returnvalue (CallFunction|GetAvatarasFortPlayerPawn))
  (bind _returnvalue_2 (CallFunction|GetAvatarasFortPlayerPawn))
  (bind _returnvalue_3 (CallFunction|GetAvatarasFortPlayerPawn))
  (Ability|Tasks|WaitGameplayTagAddToActor _returnvalue_3 "(TagName=\"MistRaven.TendrilDrain.Ascended.InEquipemtn\")"
    (:then
      (Ability|Tasks|WaitGameplayTagAddToActor _returnvalue_3 "(TagName=\"MistRaven.Connector.Ascended.InEquipment\")"
        (:then
          (Ability|Tasks|WaitGameplayTagAddToActor _returnvalue_3 "(TagName=\"MistRaven.GrimStrut.Ascended.InEquipment\")"
            (:then
              (Ability|Tasks|WaitGameplayTagRemoveFromActor _returnvalue "(TagName=\"MistRaven.TendrilDrain.Ascended.InEquipemtn\")"
                (:then
                  (Ability|Tasks|WaitGameplayTagRemoveFromActor _returnvalue "(TagName=\"MistRaven.Connector.Ascended.InEquipment\")"
                    (:then
                      (Ability|Tasks|WaitGameplayTagRemoveFromActor _returnvalue "(TagName=\"MistRaven.GrimStrut.Ascended.InEquipment\")"
                        (:Removed
                          _)))
                    (:Removed
                      _)))
                (:Removed
                  _))
              (Ability|GameplayEffect|RemoveLooseGameplayTags (CallFunction|GetAvatarasFortPlayerPawn) (Variables|Default|GetTagsAllAscendedEquipped) true))
            (:Added
              _)))
        (:Added
          _)))
    (:Added
      _))
  (if (and (GameplayTags|HasAllMatchingGameplayTags (CallFunction|GetAvatarasFortPlayerPawn) (Variables|Default|GetTagsNeededForAscension)) (not (GameplayTags|HasMatchingGameplayTag (CallFunction|GetAvatarasFortPlayerPawn) (Variables|Default|GetTagAllAscendedEquipped))))
    (Ability|GameplayEffect|AddLooseGameplayTags (CallFunction|GetAvatarasFortPlayerPawn) (Variables|Default|GetTagsAllAscendedEquipped) true)
    (if (GameplayTags|HasMatchingGameplayTag _returnvalue_2 "(TagName=\"MistRaven.AnyAscendedEquipped\")")
      (Animation|Utilities|PlayAnimMontage _returnvalue_2 "/MistRaven/Animation/AM_MistRaven_AscensionComplete.AM_MistRaven_AscensionComplete"))
    (Utilities|FlowControl|Delay 1.0)
    (Ability|GameplayEffect|RemoveLooseGameplayTags (CallFunction|GetAvatarasFortPlayerPawn) (Variables|Default|GetTagOnAscendedAchieved) true)))

(event Ability|EventOnEndAbility
  (params (bWasCancelled :type Boolean)))
