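; Gameplay ability that merges MistRaven's ascended Grim Strut and flamethrower items into the Combined Ascended item.
; On activation, if the pawn has the ascension tags and no equipped, looted or blocked tag, it applies combination and block effects, plays VFX and sound, then swaps items on authority.
; A pickup-created binding also handles drops: a dropped half is destroyed and its held counterpart replaced by the combined item.
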(event Ability|EventActivateAbility
  (bind _returnvalue (CallFunction|GetAvatarasFortPlayerPawn))
  (bind _returnvalue_2 (Fortnite|GetFortPlayerControllerfromActor _returnvalue))
  (bind _returnvalue_3 (CallFunction|GetAvatarasFortPlayerPawn))
  (bind _asplayer_controller (Utilities|Casting|CastToPlayerController (Pawn|GetController (CallFunction|GetAvatarasFortPlayerPawn))))
  (bind _asfort_player_controller (Utilities|Casting|CastToFortPlayerController _asplayer_controller))
  (EventDispatchers|BindEventtoOnPickupCreated _asfort_player_controller (delegate Custom|DroppingItem))
  (Ability|Tasks|WaitGameplayTagAddToActor _returnvalue_3 "(TagName=\"MistRaven.TendrilDrain.Ascended.InEquipemtn\")"
    (:then
      (Ability|Tasks|WaitGameplayTagAddToActor _returnvalue_3 "(TagName=\"MistRaven.GrimStrut.Ascended.InEquipment\")"
        (:Added
          _)))
    (:Added
      _))
  (if (Math|Boolean|ANDBoolean (GameplayTags|HasAllMatchingGameplayTags (CallFunction|GetAvatarasFortPlayerPawn) (Variables|Default|GetTagsNeededForAscension)) (not (GameplayTags|HasMatchingGameplayTag (CallFunction|GetAvatarasFortPlayerPawn) (Variables|Default|GetTagAllAscendedEquipped))) (not (GameplayTags|HasMatchingGameplayTag (CallFunction|GetAvatarasFortPlayerPawn) (Variables|Default|GetTagsAscendedLooted))) (not (GameplayTags|HasMatchingGameplayTag (CallFunction|GetAvatarasFortPlayerPawn) (Variables|Default|GetTagAllAscendedBlocked))))
    (Ability|ApplyGameplayEffectToOwner "/MistRaven/Gameplay/Ascended/GE_MistRavenCombination_GrantedTagsOnCombination.GE_MistRavenCombination_GrantedTagsOnCombination_C")
    (Ability|ApplyGameplayEffectToOwner "/MistRaven/Gameplay/Ascended/GE_MistRavenCombination_BlockCombination.GE_MistRavenCombination_BlockCombination_C")
    (bind _returnvalue_4 (Niagara|SpawnSystemAttached "/MistRaven/VFX/NS_Emote_MistRaven_PowerReturn.NS_Emote_MistRaven_PowerReturn" (Variables|/Script/Engine.Character|GetMesh (CallFunction|GetAvatarasFortPlayerPawn)) "root"))
    (Audio|SpawnSoundAttached "/MistRaven/Assets/Audio/MSS_MistRaven_KitComplete.MSS_MistRaven_KitComplete" _returnvalue_4)
    (Ability|HasAuthority
      (:True
        (bind _asfort_inventory_owner_interface (Utilities|Casting|CastToFortInventoryOwnerInterface (Pawn|GetController (CallFunction|GetAvatarasFortPlayerPawn))))
        (bind _returnvalue_5 (Inventory|GetInventoryItemwithTag _asfort_inventory_owner_interface (Variables|Default|GetGrimStrutWIDTag)))
        (Utilities|IsValid _returnvalue_5
          (:Is%20Valid
            _)
          (:Is%20Not%20Valid
            _))
        (Inventory|RemoveFortItemFromPlayer (Fortnite|GetFortPlayerControllerfromActor (CallFunction|GetAvatarasFortPlayerPawn)) _returnvalue_5 1 true)
        (bind _asfort_inventory_owner_interface_2 (Utilities|Casting|CastToFortInventoryOwnerInterface (Pawn|GetController (CallFunction|GetAvatarasFortPlayerPawn))))
        (bind _returnvalue_6 (InventoryManagement|FindExistingItemforDefinition _asfort_inventory_owner_interface_2 "/MistRaven/Gameplay/Ascended/RavenFlameThrowerAscended/WID_MistRave_FlameThrowerAscended.WID_MistRave_FlameThrowerAscended"))
        (Utilities|IsValid _returnvalue_6
          (:Is%20Valid
            _)
          (:Is%20Not%20Valid
            _))
        (Inventory|RemoveItemfromInventoryOwner _returnvalue_2 "/MistRaven/Gameplay/Ascended/RavenFlameThrowerAscended/WID_MistRave_FlameThrowerAscended.WID_MistRave_FlameThrowerAscended")
        (Inventory|GiveItemtoInventoryOwner _returnvalue_2 "/MistRaven/Gameplay/Ascended/CombinedAscended/WID_MistRaven_CombinedAscended.WID_MistRaven_CombinedAscended")
        (Ability|EndAbility)))))

(event Ability|EventOnEndAbility
  (params (bWasCancelled :type Boolean)))

(event Custom|DroppingItem
  (params (PlayerController :type Fort%20Player%20Controller%20Object%20Reference) (Pickup :type Fort%20Pickup%20Object%20Reference))
  (bind _returnvalue (CallFunction|GetAvatarasFortPlayerPawn))
  (bind _returnvalue_2 (Fortnite|GetFortPlayerControllerfromActor _returnvalue))
  (bind _returnvalue_3 (Pickup|GetPrimaryItemDefinition Pickup))
  (bind _returnvalue_4 (CallFunction|GetAvatarasFortPlayerPawn))
  (bind _returnvalue_5 (Fortnite|GetFortPlayerControllerfromActor _returnvalue_4))
  (if (GameplayTags|HasMatchingGameplayTag _returnvalue_3 (Variables|Default|GetGrimStrutWIDTag))
    (bind _asfort_inventory_owner_interface (Utilities|Casting|CastToFortInventoryOwnerInterface PlayerController))
    (bind _returnvalue_6 (Inventory|GetInventoryItemwithTag _asfort_inventory_owner_interface (Variables|Default|GetFlamethriwerWIDTag)))
    (Utilities|IsValid _returnvalue_6
      (:Is%20Valid
        (Actor|DestroyActor Pickup)
        (Inventory|RemoveFortItemFromPlayer _returnvalue_2 _returnvalue_6 1 true)
        (Inventory|GiveItemtoInventoryOwner _returnvalue_2 "/MistRaven/Gameplay/Ascended/CombinedAscended/WID_MistRaven_CombinedAscended.WID_MistRaven_CombinedAscended")))
    (elif (GameplayTags|HasMatchingGameplayTag _returnvalue_3 (Variables|Default|GetFlamethriwerWIDTag))
      (bind _asfort_inventory_owner_interface_2 (Utilities|Casting|CastToFortInventoryOwnerInterface PlayerController))
      (bind _returnvalue_7 (Inventory|GetInventoryItemwithTag _asfort_inventory_owner_interface_2 (Variables|Default|GetGrimStrutWIDTag)))
      (Utilities|IsValid _returnvalue_7
        (:Is%20Valid
          (Actor|DestroyActor Pickup)
          (Inventory|RemoveFortItemFromPlayer _returnvalue_5 _returnvalue_7 1 true)
          (Inventory|GiveItemtoInventoryOwner _returnvalue_5 "/MistRaven/Gameplay/Ascended/CombinedAscended/WID_MistRaven_CombinedAscended.WID_MistRaven_CombinedAscended")))))
  (Ability|ApplyGameplayEffectToOwner "/MistRaven/Gameplay/Ascended/GE_MistRavenCombination_GrantedTagsOnCombination.GE_MistRavenCombination_GrantedTagsOnCombination_C")
  (Ability|ApplyGameplayEffectToOwner "/MistRaven/Gameplay/Ascended/GE_MistRavenCombination_BlockCombination.GE_MistRavenCombination_BlockCombination_C")
  (bind _returnvalue_8 (Niagara|SpawnSystemAttached "/MistRaven/VFX/NS_Emote_MistRaven_PowerReturn.NS_Emote_MistRaven_PowerReturn" (Variables|/Script/Engine.Character|GetMesh (CallFunction|GetAvatarasFortPlayerPawn)) "root"))
  (Audio|SpawnSoundAttached "/MistRaven/Assets/Audio/MSS_MistRaven_KitComplete.MSS_MistRaven_KitComplete" _returnvalue_8)
  (Ability|EndAbility))
