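; GA_MistRaven_DeleteOnDropZeroAmmo is a MistRaven gameplay ability, named for deleting an item dropped with zero ammo.
; On activation it binds a custom ItemDropped event to the avatar Fort player pawn's OnItemDropped delegate.
; ItemDropped receives the item entry but has an empty body, so no ammo check or deletion appears here.
; The end-ability event is also empty and does not unbind the delegate.
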(event Ability|EventActivateAbility
  (FortPawn|BindEventtoOnItemDroppedDelegate (CallFunction|GetAvatarasFortPlayerPawn) (delegate Custom|ItemDropped)))

(event Custom|ItemDropped
  (params (ItemEntry :type Fort%20Item%20Entry%20Structure%20%28by%20ref%29)))

(event Ability|EventOnEndAbility
  (params (bWasCancelled :type Boolean)))
