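; Ascended MistRaven gameplay ability that previews whether the drain has a valid target, going by its name and tags.
; While active, a looping 0.1-second timer runs target selection outside placement mode.
; The hit actor's validity drives "Targettable" on two Niagara effects, and a miss can add a missing-drain loose tag.
; Each tick also removes that tag, as does ability end, which deactivates the effects and clears the timer.
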(event Custom|Tick
  (bind _visualeffect_r (Variables|Default|GetVisualEffect_R))
  (bind _visualeffect_l (Variables|Default|GetVisualEffect_L))
  (if (not (GameplayTags|HasMatchingGameplayTag (CallFunction|GetAvatarasFortPlayerPawn) "(TagName=\"MistRaven.Connector.Ascended.PlacementModeActive\")"))
    (bind (_asynctaskproxy _targetdata) (Ability|Tasks|DoTargetSelection "None" "(TagName=\"Targeting.ApplicationTag.1\")")
      (:Targeted
        (Utilities|IsValid (.HitActor (Collision|BreakHitResult (Ability|TargetData|GetHitResultfromTargetData _targetdata)))
          (:Is%20Valid
            (Niagara|SetNiagaraVariable(Bool) _visualeffect_r "Targettable" true)
            (Niagara|SetNiagaraVariable(Bool) _visualeffect_l "Targettable" true))
          (:Is%20Not%20Valid
            (Niagara|SetNiagaraVariable(Bool) _visualeffect_r "Targettable" false)
            (Niagara|SetNiagaraVariable(Bool) _visualeffect_l "Targettable" false)
            (if (not (GameplayTags|HasMatchingGameplayTag (CallFunction|GetAvatarasFortPlayerPawn) "(TagName=\"MistRaven.TendrilDrain.Ascended.NoValidTarget\")"))
              (Ability|GameplayEffect|AddLooseGameplayTags (CallFunction|GetAvatarasFortPlayerPawn) (Variables|Default|GetMissingDrainGameplayTag))))))))
  (Ability|GameplayEffect|RemoveLooseGameplayTags (CallFunction|GetAvatarasFortPlayerPawn) (Variables|Default|GetMissingDrainGameplayTag)))

(event Ability|EventActivateAbility
  (bind _returnvalue (Utilities|Time|SetTimerbyEvent (delegate Custom|Tick) 0.1 true true))
  (Variables|Default|SetUpdateTimer _returnvalue))

(event Ability|EventOnEndAbility
  (params (bWasCancelled :type Boolean))
  (Components|Activation|Deactivate (Variables|Default|GetVisualEffect_R))
  (Components|Activation|Deactivate (Variables|Default|GetVisualEffect_L))
  (Utilities|Time|ClearandInvalidateTimerbyHandle (Variables|Default|GetUpdateTimer))
  (Ability|GameplayEffect|RemoveLooseGameplayTags (CallFunction|GetAvatarasFortPlayerPawn) (Variables|Default|GetMissingDrainGameplayTag)))
